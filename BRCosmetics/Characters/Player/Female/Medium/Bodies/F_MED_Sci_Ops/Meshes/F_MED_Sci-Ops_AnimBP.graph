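# Animation Blueprint for the female medium Sci-Ops body, adding secondary motion to kneepads and hip pouches.
# It copies the attached parent mesh's pose, then blends kneepads and pouches between neighboring bones with constraints.
# Crouch events and jump/fall checks drive alphas for additive front-pouch offsets.
# Spring-based Anim Dynamics swings three pouches, with delayed external force on jumps; `pouches_force_vector` is set but never read.
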AnimBlueprint /BRCosmetics/Characters/Player/Female/Medium/Bodies/F_MED_Sci_Ops/Meshes/F_MED_Sci-Ops_AnimBP
target_skeleton=&/F_MED_Sci-Ops_Skeleton
graphs [
 AnimGraph {nodes [
   n1 = "Constraint - kneepad_l (PP)"!(
    ComponentPose=Local To Component(LocalPose=Copy Pose From Mesh() {bUseAttachedParent=true})
    Alpha=1
    ConstraintWeights_0=0.5
    ConstraintWeights_1:float=1
   ) -> (Pose) {
    alphaBoolBlend {blendOption=Linear}
    boneToModify {boneName=kneepad_l}
    constraintSetup [{targetBone {boneName=thigh_l}}, {targetBone {boneName=calf_l}}]
    constraintWeights [1,1]
   }
   n2 = "Constraint - kneepad_r (PP)"!(ComponentPose=@n1, Alpha=1, ConstraintWeights_0=0.5, ConstraintWeights_1:float=1) -> (Pose) {
    alphaBoolBlend {blendOption=Linear}
    boneToModify {boneName=kneepad_r}
    constraintSetup [{targetBone {boneName=thigh_r}}, {targetBone {boneName=calf_r}}]
    constraintWeights [1,1]
   }
   n3 = "Constraint - hip_pouch_fr_l (PP)"!(ComponentPose=@n2, Alpha=1, ConstraintWeights_0=0.6245, ConstraintWeights_1:float=0.3755) -> (Pose) {
    alphaBoolBlend {blendOption=Linear}
    boneToModify {boneName=hip_pouch_fr_l}
    constraintSetup [{targetBone {boneName=pelvis}}, {targetBone {boneName=thigh_twist_01_l}}]
    constraintWeights [1,1]
   }
   n4 = "Constraint - hip_pouch_bk_l (PP)"!(ComponentPose=@n3, Alpha=1, ConstraintWeights_0=0.6551, ConstraintWeights_1:float=0.3449) -> (Pose) {
    alphaBoolBlend {blendOption=Linear}
    boneToModify {boneName=hip_pouch_bk_l}
    constraintSetup [{targetBone {boneName=pelvis}}, {targetBone {boneName=thigh_twist_01_l}}]
    constraintWeights [1,1]
   }
   n5 = "Constraint - hip_pouch_bk_r (PP)"!(ComponentPose=@n4, Alpha=1, ConstraintWeights_0=0.6551, ConstraintWeights_1:float=0.3449) -> (Pose) {
    alphaBoolBlend {blendOption=Linear}
    boneToModify {boneName=hip_pouch_bk_r}
    constraintSetup [{targetBone {boneName=pelvis}}, {targetBone {boneName=thigh_twist_01_r}}]
    constraintWeights [1,1]
   }
   n6 = "Constraint - hip_pouch_fr_r (PP)"!(ComponentPose=@n5, Alpha=1, ConstraintWeights_0=0.6494, ConstraintWeights_1:float=0.3506) -> (Pose) {
    alphaBoolBlend {blendOption=Linear}
    boneToModify {boneName=hip_pouch_fr_r}
    constraintSetup [{targetBone {boneName=pelvis}}, {targetBone {boneName=thigh_twist_01_r}}]
    constraintWeights [1,1]
   }
   n7 = "Transform (Modify) Bone - Bone: hip_pouch_fr_l"(
    ComponentPose=@n6
    Alpha=Get crouching alpha() -> (crouching alpha:double)
    Translation="0.007,-0.148,4.192"
    Rotation="-12.561,-16.422,51.825"
    Scale="1,1,1"
   ) {
    rotationMode=BMM_Additive
    rotationSpace=BCS_BoneSpace
    translationMode=BMM_Additive
    translationSpace=BCS_BoneSpace
    alphaBoolBlend {blendOption=Linear}
    alphaScaleBiasClamp {bInterpResult=true}
    boneToModify {boneName=hip_pouch_fr_l}
    rotation=<rot -12.561 -16.4223 51.8253>
    translation=<0.007 -0.1479 4.1919>
   }
   n8 = "Transform (Modify) Bone - Bone: hip_pouch_fr_r"(
    ComponentPose=@n7
    Alpha=Get jump_alpha() -> (jump_alpha:double)
    Translation="-2.105,-0.142,-2.308"
    Rotation="-21.488,-2.346,19.141"
    Scale="1,1,1"
   ) {
    rotationMode=BMM_Additive
    rotationSpace=BCS_BoneSpace
    translationMode=BMM_Additive
    translationSpace=BCS_BoneSpace
    alphaBoolBlend {blendOption=Linear}
    alphaScaleBiasClamp {bInterpResult=true}
    boneToModify {boneName=hip_pouch_fr_r}
    rotation=<rot -21.488 -2.346 19.141>
    translation=<-2.1046 -0.142 -2.3081>
   }
   n9 = Get pouches_ext_force() -> (pouches_ext_force:Vector)
   n10 = Anim Dynamics - Bone: dyn_hip_pouch_bk_r(ComponentPose=@n8, Alpha=1, ExternalForce=@n9) {
    angularDampingOverride=1
    angularSpringConstant=1000
    bAngularSpring=true
    bGravityOverrideInSimSpace=true
    bOverrideAngularDamping=true
    alphaBoolBlend {blendOption=Linear}
    boundBone {boneName=dyn_hip_pouch_bk_r}
    componentLinearAccScale=<0.5125 0.512 0.5125>
    componentLinearVelScale=<0.5 0.5 0.5>
    physicsBodyDefinitions [{
      sphereCollisionRadius=0
      boundBone {boneName=dyn_hip_pouch_bk_r}
      boxExtents=<10 15 10>
      constraintSetup {angularLimitsMax=<1 50 1>, angularLimitsMin=<-1 -1 -1>, angularTarget=<1 0 0>}
      localJointOffset=<5 0 0>
     }]
   }
   n11 = Anim Dynamics - Bone: dyn_hip_pouch_fr_r(ComponentPose=@n10, Alpha=1, ExternalForce=@n9) {
    angularDampingOverride=1
    angularSpringConstant=1200
    bAngularSpring=true
    bGravityOverrideInSimSpace=true
    bOverrideAngularDamping=true
    alphaBoolBlend {blendOption=Linear}
    boundBone {boneName=dyn_hip_pouch_fr_r}
    componentLinearAccScale=<0.55 0.55 0.55>
    componentLinearVelScale=<0.55 0.55 0.55>
    physicsBodyDefinitions [{
      sphereCollisionRadius=0
      boundBone {boneName=dyn_hip_pouch_fr_r}
      boxExtents=<12 18 12>
      constraintSetup {angularLimitsMax=<1 40 1>, angularLimitsMin=<-1 -1 -1>, angularTarget=<1 0 0>}
      localJointOffset=<5 0 0>
     }]
   }
   n12 = Anim Dynamics - Bone: dyn_hip_pouch_bk_l(ComponentPose=@n11, Alpha=1, ExternalForce=@n9) {
    angularDampingOverride=1
    angularSpringConstant=675
    bAngularSpring=true
    bGravityOverrideInSimSpace=true
    bOverrideAngularDamping=true
    alphaBoolBlend {blendOption=Linear}
    boundBone {boneName=dyn_hip_pouch_bk_l}
    componentLinearAccScale=<0.55 0.55 0.55>
    componentLinearVelScale=<0.55 0.55 0.55>
    physicsBodyDefinitions [{
      sphereCollisionRadius=0
      boundBone {boneName=dyn_hip_pouch_bk_l}
      boxExtents=<12 18 12>
      constraintSetup {angularLimitsMax=<1 40 1>, angularLimitsMin=<-1 -1 -1>, angularTarget=<1 0 0>}
      localJointOffset=<-5 0 0>
     }]
   }
   Output Pose(Result=Component To Local(ComponentPose=@n12)) {layerGroup=None}
  ]}
 EventGraph {nodes [
   Event BlueprintUpdateAnimation()
   > n1 = Sequence()
   n2 = Get bIsJumping()
   > n3 = Branch(Condition=@n2)
   Set jump_alpha(execute=[@n3.then, @n4.then], jump_alpha:double=1) -> (Output_Get:double)
   n4 = Branch(execute=@n3.else, Condition=Get bIsFalling())
   Set jump_alpha(execute=@n4.else, jump_alpha=0)
   n5 = Branch(execute=@n1.then_1, Condition=@n2)
   > Delay(Duration=0.25)
   > Set pouches_force_vector(pouches_force_vector:Vector="0.25,2.5,0.1") -> (Output_Get:Vector)
   > Set pouches_ext_force(pouches_ext_force:Vector="0,0,1500") -> (Output_Get:Vector)
   > n6 = Delay(Duration=0.875)
   Do Once(execute=@n5.else, Reset=@n6.then)
   > Set pouches_ext_force(pouches_ext_force="0, 0, 0")
   Event OnBeginCrouching()
   > Set crouching alpha(crouching alpha:double=1) -> (Output_Get:double)
   Event OnEndCrouching()
   > Set crouching alpha(crouching alpha=0)
   Get bIsJumping()
  ]}
]
variables [
 crouching alpha {type=number}
 jump_alpha {type=number}
 pouches_ext_force {type=Vector}
 pouches_force_vector {type=Vector}
]
